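# Sound cue for the Ferret vehicle's damaged-engine fire, pairing a one-shot ignition with a looping burn.
# A mixer blends the start wave, routed to the large explosives sound class, with the loop, routed to the exterior vehicle class.
# A continuous modulator drives the loop's pitch and volume from a Fade parameter.
# The cue uses engine-loop attenuation and overrides concurrency to cap simultaneous instances.
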SoundCue /Game/Athena/Sounds/Vehicles/Valet/Vehicle_Ferret_Damage_Engine_Fire_Cue
root=Modulator_0
nodes [
 {id=Mixer_0, children [SoundClass_0, ModulatorContinuous_0], properties {inputVolume [0.75,0.75]}}
 {
  id=ModulatorContinuous_0
  children [SoundClass_1]
  properties {pitchModulationParams {minOutput=0.8, parameterName=Fade}, volumeModulationParams {maxInput=0.1, minOutput=0.5, parameterName=Fade}}
 }
 {id=Modulator_0, children [Mixer_0]}
 {id=SoundClass_0, children [WavePlayer_3], properties {soundClassOverride=/Game/Sounds/FortSoundClassesAndModes/Fort_Explosives_Large_PSM}}
 {id=SoundClass_1, children [WavePlayer_4], properties {soundClassOverride=&Fort_Vehicles_Exterior}}
 {id=WavePlayer_3, properties {soundWaveAssetPtr=&/Vehicle_Damage_Engine_Fire_Start_01}}
 {id=WavePlayer_4, properties {bLooping=true, soundWaveAssetPtr=&/Vehicle_Damage_Engine_Fire_Loop_01}}
]
properties {
 attenuationSettings=&/Vehicle_Damage_Engine_Loop_Att
 bOverrideConcurrency=true
 maxDistance=6500
 concurrencyOverrides {maxCount=2, voiceStealReleaseTime=0.15}
}
