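# Sound cue for the Ferret vehicle's damaged-engine smoke, spatialized with the engine damage loop attenuation.
# A one-shot start sound in the large-explosives sound class mixes with a looping smoke sound in the vehicle exterior class.
# An envelope fades the loop in over about the first second, beneath the start sound.
# The mix passes through a modulator and an outer envelope that holds full volume, then settles to a lower level over time.
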SoundCue /Game/Athena/Sounds/Vehicles/Valet/Vehicle_Ferret_Damage_Engine_Smoke_Cue
root=Enveloper_0
nodes [
 {
  id=Enveloper_0
  children [Modulator_0]
  properties {pitchCurve {editorCurveData {keys [{time=0}]}}, volumeCurve {editorCurveData {keys [<at 0 1>,<at 20 1>,<at 40 0.3>]}}}
 }
 {
  id=Enveloper_1
  children [WavePlayer_0]
  properties {pitchCurve {editorCurveData {keys [{time=0}]}}, volumeCurve {editorCurveData {keys [<at 0 0>,<at 0.8586 0.2529>,<at 1.25 1>]}}}
 }
 {id=Mixer_2, children [SoundClass_1,SoundClass_0], properties {inputVolume [0.75,0.75]}}
 {id=Modulator_0, children [Mixer_2]}
 {id=SoundClass_0, children [WavePlayer_1], properties {soundClassOverride=/Game/Sounds/FortSoundClassesAndModes/Fort_Explosives_Large_PSM}}
 {id=SoundClass_1, children [Enveloper_1], properties {soundClassOverride=&Fort_Vehicles_Exterior}}
 {id=WavePlayer_0, properties {bLooping=true, soundWaveAssetPtr=&/Vehicle_Damage_Engine_Smoke_Loop_01}}
 {id=WavePlayer_1, properties {soundWaveAssetPtr=&/Vehicle_Damage_Engine_Smoke_Start_01}}
]
properties {attenuationSettings=&/Vehicle_Damage_Engine_Loop_Att, maxDistance=6500}
