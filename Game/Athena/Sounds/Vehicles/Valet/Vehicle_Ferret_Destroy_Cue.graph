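# Sound cue for the Ferret vehicle's destruction explosion, layered by listener distance with a limit on concurrent plays.
# A root mixer blends close, medium and distant layers, each with its own attenuation and sound class override, weighted louder toward the distant layer.
# Each layer picks random explosion variants.
# The close layer adds slight pitch variation to the blast and mixes in a briefly delayed, randomly chosen roar tail.
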SoundCue /Game/Athena/Sounds/Vehicles/Valet/Vehicle_Ferret_Destroy_Cue
root=Mixer_8
nodes [
 {
  id=Attenuation_11
  children [Mixer_10]
  properties {attenuationSettings=/Game/Sounds/Attenuation/Fort_Explosion_Close_Athena_Att, attenuationOverrides {falloffDistance=1600}}
 }
 {
  id=Attenuation_12
  children [Random_12]
  properties {
   attenuationSettings=/Game/Sounds/Attenuation/Fort_Explosion_Medium_Athena_Att
   attenuationOverrides {falloffDistance=3000, attenuationShapeExtents {x=1000}}
  }
 }
 {
  id=Attenuation_13
  children [Random_13]
  properties {
   attenuationSettings=/Game/Sounds/Attenuation/Fort_Explosion_Distant_Athena_Att
   attenuationOverrides {falloffDistance=3000, attenuationShapeExtents {x=1000}}
  }
 }
 {id=Delay_2, children [Random_15], properties {delayMax=0.3, delayMin=0.175}}
 {id=Mixer_10, children [Modulator_5,Delay_2], properties {inputVolume [1,0.8]}}
 {id=Mixer_8, children [SoundClass_11, SoundClass_10, SoundClass_12], properties {inputVolume [0.5,1,1.5]}}
 {id=Modulator_5, children [Random_11], properties {pitchMax=1.1, pitchMin=0.9, volumeMax=1, volumeMin=1}}
 {id=Random_11, children [WavePlayer_31, WavePlayer_32], properties {bShouldExcludeFromBranchCulling=true, weights [1,1]}}
 {id=Random_12, children [WavePlayer_33, WavePlayer_34], properties {bShouldExcludeFromBranchCulling=true, weights [1,1]}}
 {id=Random_13, children [WavePlayer_35, WavePlayer_36], properties {bShouldExcludeFromBranchCulling=true, weights [1,1]}}
 {
  id=Random_15
  children [WavePlayer_43, WavePlayer_44, WavePlayer_45, WavePlayer_46, WavePlayer_47]
  properties {bShouldExcludeFromBranchCulling=true, weights [1,1,1,1,1]}
 }
 {
  id=SoundClass_10
  children [Attenuation_12]
  properties {soundClassOverride=&Fort_WeaponRanged_Fire_Distant}
 }
 {id=SoundClass_11, children [Attenuation_11], properties {soundClassOverride=/Game/Sounds/FortSoundClassesAndModes/Fort_Explosives_Large_PSM}}
 {
  id=SoundClass_12
  children [Attenuation_13]
  properties {soundClassOverride=&Fort_WeaponRanged_Fire_Distant}
 }
 {id=WavePlayer_31, properties {soundWaveAssetPtr=&/Vehicle_Damage_Explode_Close_02}}
 {id=WavePlayer_32, properties {soundWaveAssetPtr=&/Vehicle_Damage_Explode_Close_01}}
 {id=WavePlayer_33, properties {soundWaveAssetPtr=&/Vehicle_Damage_Explode_Med_01}}
 {id=WavePlayer_34, properties {soundWaveAssetPtr=&/Vehicle_Damage_Explode_Med_02}}
 {id=WavePlayer_35, properties {soundWaveAssetPtr=&/Vehicle_Damage_Explode_Distant_01}}
 {id=WavePlayer_36, properties {soundWaveAssetPtr=&/Vehicle_Damage_Explode_Distant_02}}
 {id=WavePlayer_43, properties {soundWaveAssetPtr=&/Vehicle_Damage_Explode_Close_Roar_05}}
 {id=WavePlayer_44, properties {soundWaveAssetPtr=&/Vehicle_Damage_Explode_Close_Roar_04}}
 {id=WavePlayer_45, properties {soundWaveAssetPtr=&/Vehicle_Damage_Explode_Close_Roar_03}}
 {id=WavePlayer_46, properties {soundWaveAssetPtr=&/Vehicle_Damage_Explode_Close_Roar_02}}
 {id=WavePlayer_47, properties {soundWaveAssetPtr=&/Vehicle_Damage_Explode_Close_Roar_01}}
]
properties {bOverrideConcurrency=true, maxDistance=27500, volumeMultiplier=1, concurrencyOverrides {maxCount=3}}
